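# Post-process-style Anim Blueprint for the Camila female head on the Fortnite base head skeleton, set up for secondary motion on hair and an earring.
# A chain of AnimDynamics nodes covers side, front and top hair bones plus the earring, using limits, damping and some angular springs.
# That chain and a RigidBody node never reach the output, so Input Pose passes through unchanged.
# The event graph holds only unconnected stubs.
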AnimBlueprint /Game/Characters/Player/Female/Small/Heads/F_SML_HIS_Camila_Head/Mesh/F_SML_HIS_Camila_Head_01_Dyn_AnimBP
target_skeleton=/Game/Characters/Player/Common/Fortnite_Base_Head/Fortnite_Base_Head_Export_Skeleton
graphs [
 AnimGraph {nodes [
   n1 = Anim Dynamics - Bone: dyn_hair_l(Alpha=1, ExternalForce="0,0,0") {
    angularDampingOverride=0.1
    angularSpringConstant=10
    bAngularSpring=true
    bGravityOverrideInSimSpace=true
    bOverrideAngularDamping=true
    bOverrideLinearDamping=true
    bUsePlanarLimit=false
    gravityScale=0.33
    linearDampingOverride=0.5
    numSolverIterationsPostUpdate=4
    numSolverIterationsPreUpdate=16
    alphaBoolBlend {blendOption=Linear}
    boundBone {boneName=dyn_hair_l}
    physicsBodyDefinitions [{
      sphereCollisionRadius=0
      boundBone {boneName=dyn_hair_l}
      boxExtents=<5 5 5>
      constraintSetup {
       angularTargetAxis=AxisY
       coneAngle=45
       angularLimitsMax=<0.001 20 20>
       angularLimitsMin=<-0.001 -20 -10>
       angularTarget=<0 0 0>
       linearAxesMax=<0 0 0.5>
       linearAxesMin=<0 0 -0.5>
      }
      localJointOffset=<4 0.001 -2>
     }]
   }
   n2 = Anim Dynamics - Bone: dyn_hair_r(ComponentPose=@n1, Alpha=1, ExternalForce="0,0,0") {
    angularDampingOverride=0.1
    angularSpringConstant=10
    bAngularSpring=true
    bGravityOverrideInSimSpace=true
    bOverrideAngularDamping=true
    bOverrideLinearDamping=true
    gravityScale=0.33
    linearDampingOverride=0.5
    numSolverIterationsPostUpdate=4
    numSolverIterationsPreUpdate=16
    alphaBoolBlend {blendOption=Linear}
    boundBone {boneName=dyn_hair_r}
    physicsBodyDefinitions [{
      sphereCollisionRadius=0
      boundBone {boneName=dyn_hair_r}
      boxExtents=<5 5 5>
      constraintSetup {
       angularTargetAxis=AxisY
       coneAngle=45
       angularLimitsMax=<0.001 20 20>
       angularLimitsMin=<-0.001 -20 -5>
       angularTarget=<0 0 0>
       linearAxesMax=<0 0 0.5>
       linearAxesMin=<0 0 -0.5>
      }
      localJointOffset=<-4 -0.001 2>
     }]
   }
   n3 = Anim Dynamics - Bone: dyn_hair_front_a(ComponentPose=@n2, Alpha=1, ExternalForce="0,0,0") {
    angularDampingOverride=0.99
    angularSpringConstant=10
    bGravityOverrideInSimSpace=true
    bOverrideAngularDamping=true
    bUsePlanarLimit=false
    alphaBoolBlend {blendOption=Linear}
    boundBone {boneName=dyn_hair_front_a}
    chainEnd {boneName=hair_front_a_end}
    physicsBodyDefinitions [{
      sphereCollisionRadius=0
      boundBone {boneName=dyn_hair_front_a}
      boxExtents=<15 8 5>
      constraintSetup {angularLimitsMax=<5 10 10>, angularLimitsMin=<0 0 0>, angularTarget=<1 0 0>}
      localJointOffset=<7 0 0>
     }]
   }
   n4 = Anim Dynamics - Bone: dyn_hair_front_b(ComponentPose=@n3, Alpha=1, ExternalForce="0,0,0") {
    angularDampingOverride=0.99
    angularSpringConstant=10
    bGravityOverrideInSimSpace=true
    bOverrideAngularDamping=true
    bUsePlanarLimit=false
    alphaBoolBlend {blendOption=Linear}
    boundBone {boneName=dyn_hair_front_b}
    chainEnd {boneName=hair_front_b_end}
    physicsBodyDefinitions [{
      sphereCollisionRadius=0
      boundBone {boneName=dyn_hair_front_b}
      boxExtents=<10 5 5>
      constraintSetup {angularLimitsMax=<0 10 10>, angularLimitsMin=<0 0 0>, angularTarget=<0 0 0>}
      localJointOffset=<4 0 0>
     }]
   }
   n5 = Anim Dynamics - Bone: dyn_hair_top(ComponentPose=@n4, Alpha=1, ExternalForce="0,0,0") {
    angularDampingOverride=0.99
    angularSpringConstant=10
    bAngularSpring=true
    bGravityOverrideInSimSpace=true
    bUsePlanarLimit=false
    alphaBoolBlend {blendOption=Linear}
    boundBone {boneName=dyn_hair_top}
    chainEnd {boneName=hair_front_a_end}
    physicsBodyDefinitions [{
      sphereCollisionRadius=0
      boundBone {boneName=dyn_hair_top}
      boxExtents=<3 15 15>
      constraintSetup {angularLimitsMax=<0 10 0>, angularLimitsMin=<0 -5 0>, angularTarget=<1 0 0>, linearAxesMax=<5 0 0>}
      localJointOffset=<-1.5 0 0>
     }]
   }
   Anim Dynamics - Bone: dyn_earring(ComponentPose=@n5, Alpha=1, ExternalForce="0,0,0") {
    angularDampingOverride=0.99
    angularSpringConstant=10
    bAngularSpring=true
    bGravityOverrideInSimSpace=true
    bUsePlanarLimit=false
    alphaBoolBlend {blendOption=Linear}
    boundBone {boneName=dyn_earring}
    chainEnd {boneName=hair_front_a_end}
    physicsBodyDefinitions [{
      sphereCollisionRadius=0
      boundBone {boneName=dyn_earring}
      boxExtents=<5 5 2>
      constraintSetup {angularLimitsMax=<0 0 60>, angularLimitsMin=<0 0 -60>, angularTarget=<1 0 0>, linearAxesMax=<5 0 0>}
      localJointOffset=<-2 0 0>
     }]
   }
   Component To Local(ComponentPose=Local To Component())
   RigidBody() {alphaBoolBlend {blendOption=Linear}}
   AnimGraphNode_SubInput_0 = Input Pose()
   Output Pose(Result=@AnimGraphNode_SubInput_0) {layerGroup=None}
  ]}
 EventGraph {nodes [
   TryGetPawnOwner()
   Event BlueprintUpdateAnimation()
  ]}
]
variables []
